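; Actor Blueprint in a BattlePassS22 quest review tool folder, named as a holder of changes for review; it shows no gameplay behavior.
; ANewFunction branches on the negation of a default-named local variable, with no actions attached.
; The construction script only calls the parent implementation.
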(fn ANewFunction
  (if (not (LocalVariables|GetANewLocalVariable))
    _))

(fn ConstructionScript
  (|CallParentFunction))
