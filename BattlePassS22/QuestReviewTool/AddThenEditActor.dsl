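; AddThenEditActor is an actor Blueprint in a quest review tool folder that holds only stub logic, with no clear gameplay purpose.
; AddAFunction reads a local boolean named ANewLocalVariable and branches on it, with no action on the branch.
; The construction script only calls the parent implementation.
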(fn AddAFunction
  (if (LocalVariables|GetANewLocalVariable)
    _))

(fn ConstructionScript
  (|CallParentFunction))
