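; Blueprint actor in a BattlePassS22 QuestReviewTool folder, apparently a scratch asset for exercising added, changed and checked-in states, judging by its name.
; Its construction script only calls the parent implementation.
; A function named AddANewFunction runs a DoOnce node fed by a placeholder local variable, with nothing connected after it.
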(fn AddANewFunction
  (Utilities|FlowControl|DoOnce (LocalVariables|GetANewLocalVariable)))

(fn ConstructionScript
  (|CallParentFunction))
