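; Child actor Blueprint in the QuestReviewTool folder that holds minimal test-like logic with no clear gameplay purpose.
; AFunction runs a while loop on a local variable with an empty body, then prints "FinishedLooping".
; Nothing in the function changes that variable, so the loop never exits if the condition starts true.
; The construction script only calls its parent.
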(fn AFunction
  (while (LocalVariables|GetALocalVariable)
    _)
  (Development|PrintString "FinishedLooping"))

(fn ConstructionScript
  (|CallParentFunction))
